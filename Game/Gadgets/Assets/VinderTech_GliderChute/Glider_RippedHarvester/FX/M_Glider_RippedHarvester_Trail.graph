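# Unlit, two-sided translucent trail material for the Ripped Harvester glider's Niagara ribbon and mesh particles.
# Two counter-panning Perlin samples multiply into a distortion that warps scrolling lava noise, strongest at low U.
# UV falloffs and a sine across V shape that noise into a mask blending two colors, each switched to alternates by the Variant dynamic parameter.
# Particle color tints the result; particle alpha, the mask and double-sided Fresnel form an opacity term.
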Material /Game/Gadgets/Assets/VinderTech_GliderChute/Glider_RippedHarvester/FX/M_Glider_RippedHarvester_Trail
graph {
 properties {
  bUsedWithNiagaraMeshParticles=true
  bUsedWithNiagaraRibbons=true
  blendMode=BLEND_Translucent
  pixelDepthOffsetMode=PDOM_Legacy
  shadingModel=MSM_Unlit
  twoSided=true
  usedShadingModels=Unlit
 }
 nodes [
  n1 = TextureCoordinate()
  MaterialFunctionCall(Float2=@n1) {materialFunction=&BreakOutFloat2Components}
  n2 = DynamicParameter() {defaultValue=<rgba 0 1 1 1>, paramNames [Variant,Param2,Param3,Param4]}
  n3 = LinearInterpolate(
   A=VectorParameter() {parameterName=Color1, defaultValue=<rgba 1 0.0017 0 1>}.RGB
   B=VectorParameter() {parameterName=ColorAlt1, defaultValue=<rgba 0 0.0833 1 1>}.RGB
   Alpha=@n2.Variant
  )
  n4 = LinearInterpolate(
   A=VectorParameter() {parameterName=Color2, defaultValue=<rgba 10 1.8733 0.8 1>}.RGB
   B=VectorParameter() {parameterName=ColorAlt2, defaultValue=<rgba 0.2 1.8333 10 1>}.RGB
   Alpha=@n2.Variant
  )
  n5 = MaterialFunctionCall(Float2=@n1) {materialFunction=&BreakOutFloat2Components}
  n6 = OneMinus(None=@n5.R)
  n7 = ScalarParameter() {defaultValue=1.3, parameterName=ColorPower}
  n8 = ScalarParameter() {defaultValue=3, parameterName=OpacityPower}
  n9 = AppendVector(A=Multiply(A=@n5.R), B=Multiply(A=@n5.G))
  n10 = AppendVector(A=Multiply(A=@n5.R) {constB=2}, B=Multiply(A=@n5.G))
  n11 = TextureSampleParameter2D(UVs=Panner(Coordinate=@n10) {speedX=0.2, speedY=-0.4}) {
   parameterName=TexDistort
   samplerType=SAMPLERTYPE_Masks
   texture=/Game/Characters/Player/Female/Medium/Bodies/F_MED_Neon_Tight_Suit/Materials/Soft_Perlin_output
  }
  n12 = TextureSampleParameter2D(UVs=Panner(Coordinate=@n10) {speedX=-0.2, speedY=0.3}) {
   parameterName=TexDistort
   samplerType=SAMPLERTYPE_Masks
   texture=&Soft_Perlin_output
  }
  n13 = LinearInterpolate(Alpha=Multiply(A=@n11.R, B=@n12.R)) {constA=0.1}
  n14 = Multiply(A=@n13) {constB=0.2}
  n15 = Saturate(None=Power(Base=OneMinus(None=@n5.R)) {constExponent=4})
  n16 = TextureSampleParameter2D(UVs=Add(A=Panner(Coordinate=@n9) {speedX=-1.3}, B=Multiply(A=@n14, B=@n15))) {
   parameterName=Tex1
   samplerType=SAMPLERTYPE_LinearGrayscale
   texture=/Game/Weapons/FORT_Melee/Pickaxe_RenegadeRaider_Spark/FX/Materials/T_Pickaxe_RenegadeRaider_Spark_LavaNoise
  }
  n17 = Multiply(
   A=Saturate(None=Power(Base=@n6, Exp=@n7) {constExponent=1.3})
   B=LinearInterpolate(A=Power(Base=@n6, Exp=@n8) {constExponent=3}, B=LinearInterpolate(A=@n16.R), Alpha=OneMinus(None=@n5.R)) {constAlpha=1}
  )
  n18 = Sine(None=Multiply(A=@n5.G) {constB=0.5})
  n19 = Add(A=Saturate(None=Power(Base=@n18) {constExponent=4}), B=Multiply(A=Multiply(A=@n13), B=Power(Base=@n18) {constExponent=1.5}))
  n20 = Saturate(None=Power(Base=Multiply(A=@n17, B=@n19)))
  n21 = ParticleColor()
  MaterialFunctionCall(Emissive=Multiply(A=LinearInterpolate(A=@n3, B=@n4, Alpha=@n20), B=@n21.RGB)) {materialFunction=&MF_FX_EmissiveAdjustments}
  n22 = Saturate(None=Multiply(A=@n5.R) {constB=8})
  n23 = Saturate(None=Multiply(A=@n19, B=Multiply(A=@n17, B=@n22)) {constA=1})
  n24 = VertexNormalWS()
  Multiply(A=Multiply(A=Power(Base=@n23) {constExponent=1}, B=@n24)) {constB=4}
  Multiply(
   A=Multiply(A=@n21.A, B=@n23)
   B=MaterialFunctionCall(
    Brightness=ScalarParameter() {defaultValue=6, parameterName=Fresnel_Brightness}
    Exponent=ScalarParameter() {defaultValue=3, parameterName=Fresnel_Exponent}
    Camera Vector=CameraVectorWS()
    Vertex / Pixel Normal=PixelNormalWS()
   ) {materialFunction=&MF_Fresnel_FX}.Fresnel Double Sided
  )
  Normalize(VectorInput=@n24)
  Power(Base=@n5.R) {constExponent=1.5}
  Reroute(None=@n24)
  TransformPosition(None=@n24) {transformType=TRANSFORMPOSSOURCE_World}
  Transform(None=@n24)
  VertexNormalWS()
 ]
}
